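# Additive unlit Niagara sprite material that draws a tapered light beam for the Lights aura cosmetic, tinted by particle color with a whitened core.
# UV math forms the soft cone, changing base thickness when viewed along the instance Z axis.
# Panning, self-distorted smoke in object-stabilized screen-space UVs optionally modulates the beam and adds glow and hot edges.
# Length gradient and view-facing fades shape opacity; Low and Medium quality skip depth and camera fades.
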Material /SparksCosmetics/Items/Auras/Assets/NS_Aura_Lights/M_Aura_Lights_LightBeam
graph {
 properties {bUsedWithNiagaraSprites=true, blendMode=BLEND_Additive, pixelDepthOffsetMode=PDOM_Legacy, shadingModel=MSM_Unlit, usedShadingModels=Unlit}
 nodes [
  n1 = ParticleColor()
  n2 = ScalarParameter() {defaultValue=0.1, parameterName=White Core Amount}
  n3 = ScalarParameter() {defaultValue=5, parameterName=White Core Sharpness}
  n4 = TextureCoordinate()
  n5 = TextureCoordinate()
  n6 = ScalarParameter() {defaultValue=0.133, parameterName=Base Thickness Sides}
  n7 = ScalarParameter() {defaultValue=0.03, parameterName=Base Thickness Above}
  n8 = Constant3Vector() {constant=<rgba 0 0 1 0>}
  n9 = Abs(None=DotProduct(A=Transform(None=@n8) {transformSourceType=TRANSFORMSOURCE_Instance}, B=CameraVectorWS()))
  n10 = OneMinus(None=Saturate(None=LinearInterpolate(A=@n6, B=@n7, Alpha=@n9)))
  n11 = ComponentMask(None=@n5) {g=true}
  n12 = OneMinus(None=@n11)
  n13 = Divide(A=ConstantBiasScale(None=ComponentMask(None=@n5) {r=true}) {bias=-0.5, scale=2}, B=OneMinus(None=Multiply(A=@n10, B=@n12)))
  n14 = Saturate(None=DotProduct(A=@n13, B=@n13))
  n15 = OneMinus(None=@n14)
  n16 = Multiply(A=OneMinus(None=ComponentMask(None=@n4) {g=true}), B=@n15)
  MaterialFunctionCall(Emissive=LinearInterpolate(A=@n1.RGB, B=LinearInterpolate(A=@n1.RGB, Alpha=@n2) {constAlpha=0.1}, Alpha=Power(Base=@n16, Exp=@n3)) {constA=1}) {materialFunction=&MF_FX_EmissiveAdjustments}
  n17 = Power(Base=Saturate(None=SquareRoot(None=@n15)), Exp=ScalarParameter() {defaultValue=5, parameterName=Edge Falloff})
  n18 = NamedRerouteUsage()
  n19 = Multiply(A=@n17, B=ScalarParameter() {desc=Only works when using smoke noise, parameterName=Glow Amount}) {constB=0}
  n20 = MaterialFunctionCall!(Gradient=@n11, Center Point=ScalarParameter() {defaultValue=0.05, parameterName=Base Fade}) {materialFunction=/Game/Effects/Fort_Effects/MaterialFunctions/MF_GradientReflected}
  n21 = Saturate(None=Multiply(A=Multiply(A=@n17, B=@n14), B=Multiply(A=@n20.Result, B=@n20.Result)))
  n22 = Multiply(A=Multiply(A=Power(Base=@n21)) {constB=10}, B=ScalarParameter() {desc=Only works when using smoke noise, parameterName=Hot Edge Amount})
  n23 = LinearInterpolate(A=@n17, B=Add(A=Multiply(A=@n17, B=@n18), B=Add(A=@n19, B=@n22)), Alpha=ScalarParameter() {defaultValue=0.1, parameterName=Use Smoke})
  n24 = Abs(None=DotProduct(A=VertexNormalWS(), B=CameraVectorWS()))
  n25 = ScalarParameter() {defaultValue=50, parameterName=Camera Fade Length}
  n26 = ScalarParameter() {parameterName=Camera Fade Offset}
  n27 = DepthFade(
   Opacity=Multiply(
    A=@n20.Result
    B=MaterialFunctionCall(Fade Length=@n25, Fade Offset=@n26) {materialFunction=&CameraDepthFade}.Result
   )
   FadeDistance=ScalarParameter() {defaultValue=33, parameterName=Depth Fade Distance}
  )
  n28 = QualitySwitch(Default=@n27, Low=@n20.Result, High=@n27, Medium=@n20.Result, Epic=@n27)
  n29 = ScalarParameter() {defaultValue=2.666, parameterName=Length Falloff}
  Multiply(A=ParticleColor().A, B=Multiply(A=Multiply(A=@n23, B=@n24), B=Power(Base=@n28, Exp=@n29)))
  n30 = MaterialFunctionCall!(Position, UVRadialScale=ScalarParameter() {defaultValue=0.1, group=World Space Smoke, parameterName=WS Smoke Noise Scale}) {materialFunction=/Game/Weapons/WeaponSkins/MaterialFunction/ObjectStabilizedScreenSpaceUVs}
  n31 = Multiply(A=ScalarParameter() {defaultValue=1, group=World Space Smoke, parameterName=Offset Smoke Per Particle}, B=ParticleRandom())
  n32 = ScalarParameter() {defaultValue=0.2, group=World Space Smoke, parameterName=WS Smoke Speed X}
  n33 = Multiply(A=@n32) {constB=-1.15}
  n34 = ScalarParameter() {defaultValue=0.5, group=World Space Smoke, parameterName=WS Smoke Speed Y}
  n35 = Multiply(A=@n34) {constB=1.2}
  n36 = TextureSampleParameter2D(UVs=Panner(Coordinate=Add(A=@n30.StabilizedUVs, B=@n31), Speed=AppendVector(A=@n33, B=@n35)) {bFractionalPart=true, speedX=0.09, speedY=-0.1}) {group=World Space Smoke, parameterName=WS Smoke Noise, texture=/Game/Effects/Fort_Effects/Textures/Tile/T_Inky_Smoke_Tile}
  n37 = Add(A=AppendVector(A=@n36.R, B=@n36.R)) {constB=-0.5}
  n38 = ScalarParameter() {defaultValue=0.2, group=World Space Smoke, parameterName=Distort Smoke UVs}
  n39 = Panner(Coordinate=Add(A=@n30.StabilizedUVs, B=Multiply(A=@n37, B=@n38) {constB=0.2}), Speed=AppendVector(A=@n32, B=@n34)) {bFractionalPart=true, speedX=0.09, speedY=-0.1}
  NamedRerouteDeclaration(None=SmoothStep(
    Min=ScalarParameter() {group=World Space Smoke, parameterName=WS Smoke Min}
    Max=ScalarParameter() {defaultValue=0.3, group=World Space Smoke, parameterName=WS Smoke Max}
    Value=TextureSampleParameter2D(UVs=@n39) {group=World Space Smoke, parameterName=WS Smoke Noise, texture=&T_Inky_Smoke_Tile}.G
   ) {constMax=0.5}) {name=WorldSpaceSmoke, nodeColor=<rgba 0.4059 0.6426 1 1>}
 ]
}
