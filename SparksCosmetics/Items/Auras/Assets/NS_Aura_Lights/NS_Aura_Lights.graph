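# Niagara aura cosmetic that surrounds a character with colored stage-style lights, glows and sparkles.
# MainLight aims a beam at a smoothed bone position and sets light parameters there, while two FigureEightBeams sweep figure-eight targets.
# Body_Stars, Dust and Glows sample skeletal bones, with star lifetime and motion responding to bone velocity.
# System.MainColor tints glows and main light; IsFrontEnd, IsLargeBody and time of day adjust brightness or size.
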NiagaraSystem /SparksCosmetics/Items/Auras/Assets/NS_Aura_Lights/NS_Aura_Lights
dependencies {
 usedDataInterfaces [
  &NiagaraDataInterfaceCurve
  &NiagaraDataInterfaceArrayFloat3
  &NiagaraDataInterfaceColorCurve
  &NiagaraDataInterfaceVector2DCurve
  &NiagaraDataInterfaceArrayInt32
  &NiagaraDataInterfaceArrayFloat
  &NiagaraDataInterfaceArrayColor
  &NiagaraDataInterfaceSkeletalMesh
 ]
 usedDynamicInputs [&MakeVector, &RandomRangeLinearColor]
 usedModules [
  &self:SystemSpawnScript.SystemScriptSource.SystemScriptGraph.NiagaraNodeAssignment_0.SetVariables_#1
  &self:SystemSpawnScript.SystemScriptSource.SystemScriptGraph.NiagaraNodeAssignment_1.SetVariables_#2
  &self:SystemSpawnScript.SystemScriptSource.SystemScriptGraph.NiagaraNodeAssignment_2.SetVariables_#3
  &SystemState
  &self:SystemSpawnScript.SystemScriptSource.SystemScriptGraph.NiagaraNodeAssignment_3.SetVariables_#4
  &self:SystemSpawnScript.SystemScriptSource.SystemScriptGraph.NiagaraNodeAssignment_4.SetVariables_#5
 ]
}
emitters {
 BlackBacker {
  bEnabled=false
  emitterUpdateScript [
   &EmitterState()
   &SpawnBurst_Instantaneous(Spawn Count=3, Use Spawn Probability=false, UseLoopCountLimit=true)
  ]
  particleSpawnScript [
   &InitializeParticle(
    Color=&SelectLinearColorFromArray(
     Array Sampling Mode=::NewEnumerator1
     Direct Array Index=&ReturnExecIndex()
     Color Selection Array=DataInterface() {propertyValues {ColorData [<rgba 0.1009 0.0268 0.0268 0.75>, <rgba 0.028 0.0305 0.1287 0.5>]}}
     Random Seed=?
    )
    Lifetime=5
    Position Offset=&SelectVectorFromArray(
     Direct Array Index=&ReturnExecIndex()
     Vector Selection Array=DataInterface() {propertyValues {FloatData [<30 -30 200>, <40 -35 150>, <-40 -35 150>]}}
    )
    Sprite Size Mode=::NewEnumerator3
    Uniform Sprite Size=&SelectFloatFromArray(
     Direct Array Index=&ReturnExecIndex()
     Float Selection Array=DataInterface() {propertyValues {FloatData [150,100]}}
    )
    UsePositionOffset=true
    Write Material Random=true
   )
   &CameraOffset(Camera Offset Amount=-10)
   &DynamicMaterialParameters(
    Index 0 Param 2=0.1
    Module.Index0Float0Write=true
    Module.Index0Float1Write=true
    Module.Index0Float2Write=true
    Module.Index0Float3Write=true
   )
  ]
  particleUpdateScript [&ParticleState(Kill Particles When Lifetime Has Elapsed=false)]
  rendererClasses [&NiagaraSpriteRendererProperties]
  renderers [{rendererClass=&NiagaraSpriteRendererProperties}]
 }
 Body_Stars {
  emitterUpdateScript [
   &EmitterState()
   &SpawnRate(SpawnRate=55, Use Spawn Probability=false)
  ]
  particleSpawnScript [
   /SetVariables(Particles.PickBone:NiagaraInt32=&RandomRangeInt(Minimum=1, Maximum=6)) {isSetParameters=true}
   /SetVariables(Particles.PickBoneMatch:NiagaraInt32=&SelectIntFromArray(Direct Array Index=$Particles.PickBone, Int Selection Array=DataInterface() {propertyValues {IntData [0,0,0,1,2,3,4]}})) {isSetParameters=true}
   /SetVariables(Transient.Variance:NiagaraFloat=&RandomRangeFloat(Minimum=0, Maximum=1)) {isSetParameters=true}
   &SkeletalMeshLocation(
    Bone / Socket Index=$Particles.PickBone
    Bone Sampling Mode=::NewEnumerator4
    MaskSpawnGroup=false
    Orientation Sampling=::NewEnumerator1
    Position Sampling=::NewEnumerator2
    Sampled Position Offset=&RandomVector(Vector Scale=5)
    Sampled Position Offset Coordinate Space:ENiagaraSkelMeshTransforms=::NewEnumerator1
    Skeletal Mesh=$System.Skeletal Mesh
   )
   &SkeletalMeshLocation(
    Bone / Socket Index=$Particles.PickBoneMatch
    Bone Sampling Mode=::NewEnumerator4
    MaskSpawnGroup=false
    Orientation Sampling=::NewEnumerator1
    Position Sampling=::NewEnumerator2
    Sampled Position Offset=&RandomVector(Vector Scale=5)
    Sampled Position Offset Coordinate Space:ENiagaraSkelMeshTransforms=::NewEnumerator1
    Skeletal Mesh=$System.Skeletal Mesh
   ) {module=SkeletalMeshLocation001}
   &SkeletalMeshLocation(
    Bone / Socket Index=7
    Bone Sampling Mode=::NewEnumerator4
    MaskSpawnGroup=false
    Orientation Sampling=::NewEnumerator1
    Position Sampling=::NewEnumerator2
    Skeletal Mesh=$System.Skeletal Mesh
   ) {module=SkeletalMeshLocation002}
   /SetVariables(Transient.VelocityModifier:NiagaraFloat=&RemapRange(
     Input Value To Remap=&VectorLength(Vector=$Particles.SkeletalMeshLocation.SampledVelocity)
     Input Min=15
     Input Max=100
     Output Min=0
     Output Max=1
    )) {isSetParameters=true}
   &InitializeParticle(
    Color=&MultiplyLinearColor(
     Linear Color=&ColorFromCurve(
      ColorCurve=DataInterface() {propertyValues {
        AlphaCurve {keys [<at 0 1 RCIM_Cubic>]}
        BlueCurve {keys [<at 0 1>, {arriveTangent=-2, interpMode=RCIM_Cubic, leaveTangent=-2, time=0.25, value=0}, <at 0.5 0>, <at 0.75 0>, <at 1 1 RCIM_Cubic>]}
        GreenCurve {keys [<at 0 0>, {arriveTangent=0.6, interpMode=RCIM_Cubic, leaveTangent=0.6, time=0.25, value=0}, <at 0.5 0.3>, <at 0.75 0.3>, <at 1 0 RCIM_Cubic>]}
        RedCurve {keys [<at 0 0>, {arriveTangent=0.6, interpMode=RCIM_Cubic, leaveTangent=0.6, time=0.25, value=0.6}, <at 0.5 0.3>, <at 0.75 0>, <at 1 0 RCIM_Cubic>]}
       }}
      CurveIndex=&RandomRangeFloat(Minimum=0, Maximum=1)
     )
     Scale Factor=&MakeFloat4FromVectorAndFloat(
      "Vector (RGB)"=&VectorFromFloat(Value=&RandomRangeFloat(Minimum=1, Maximum=3))
      "Float (Alpha)"=1
     )
    )
    Lifetime=&Multiply_Float(
     A=&RandomRangeFloat(Minimum=0.3, Maximum=&Lerp_Float(A=1.5, B=0.5, Alpha=$Transient.VelocityModifier))
     B=&Lerp_Float(A=1, B=0.5, Alpha=$Transient.Variance)
    )
    Mesh Scale Mode=::NewEnumerator3
    Mesh Uniform Scale=&RandomRangeFloat(Minimum=0.05, Maximum=0.25)
    Position=&LerpPosition(
     A=$Particles.SkeletalMeshLocation.SampledPosition
     B=$Particles.SkeletalMeshLocation001.SampledPosition
     Alpha=&RandomRangeFloat(Minimum=0, Maximum=1)
    )
    Position Mode=::NewEnumerator1
    Sprite Size Mode=::NewEnumerator3
    Uniform Sprite Size=&Lerp_Float(
     A=3
     B=12
     Alpha=&FloatFromCurve(
      FloatCurve=DataInterface() {propertyValues {Curve {keys [
          {arriveTangent=0.5641, interpMode=RCIM_Cubic, leaveTangent=0.5641, tangentMode=RCTM_User, time=0, value=0}
          {arriveTangent=1, interpMode=RCIM_Cubic, leaveTangent=1, time=0.7913, value=0.5623}
          {arriveTangent=0.9829, tangentMode=RCTM_User, time=1, value=1}
         ]}}}
      CurveIndex=$Transient.Variance
     )
    )
    UsePositionOffset=false
    Write Material Random=true
   )
   &InheritVelocity(
    Inherited Velocity Amount Scale=<1 1 0.2>
    Speed Limit=true
    UseSpeedThreshold=false
    Velocity Source=$Particles.SkeletalMeshLocation.SampledVelocity
   )
   &AddVelocity(
    Rotation Coordinate Space=::World
    Velocity=&Add_Vector(
     A=&RandomRangeVector(Minimum=<-3 -3 15>, Maximum=<3 3 38>)
     B=&Multiply_VectorByFloat(
      Vector=&NormalizeVector(
       Vector To Normalize=&Subtract_Position(A=$Particles.Position, B=$Particles.SkeletalMeshLocation002.SampledPosition)
       Safe Normalize=true
       Safe Fallback Threshold=0
       Safe Fallback Vector=<0 0 0.001>
      )
      Float=&RandomRangeFloat(Minimum=5, Maximum=20)
     )
    )
    Velocity Speed Scale=&Lerp_Float(A=1, B=0.5, Alpha=$Transient.VelocityModifier)
   )
   /SetVariables(Particles.Color=&MakeCustomLinearColorFromBool(
     Bool=$User.IsFrontEnd
     True Color=$Particles.Color
     False Color=&MultiplyLinearColor(
      Linear Color=&HueShiftLinearColor(Hue=0, Saturation=2, Value=0.5, Linear Color=$Particles.Color)
      Scale Factor=&MakeFloat4FromVectorAndFloat("Vector (RGB)"=<1 1 1>, "Float (Alpha)"=3)
     )
    )) {isSetParameters=true}
   /SetVariables(Particles.Color=&ColorFromCurve(
     ColorCurve=DataInterface() {propertyValues {
       AlphaCurve {keys [<at 0 1 RCIM_Cubic>]}
       BlueCurve {keys [<at 0 1 RCIM_Cubic>,<at 0.02 0>,<at 0.98 1>,<at 1 0 RCIM_Cubic>]}
       GreenCurve {keys [<at 0 1 RCIM_Cubic>,<at 0.02 0>,<at 0.98 0>,<at 1 1 RCIM_Cubic>]}
       RedCurve {keys [<at 0 1 RCIM_Cubic>,<at 0.02 1>,<at 0.98 0>,<at 1 0 RCIM_Cubic>]}
      }}
     CurveIndex=$Transient.VelocityModifier
    )) {enabled=false, isSetParameters=true}
  ]
  particleUpdateScript [
   &CameraOffset(Camera Offset Amount=-25)
   &ParticleState()
   &SpriteRotationRate(Rotation Rate=&FloatFromCurve(
     FloatCurve=DataInterface() {propertyValues {Curve {keys [<at 0 1 RCIM_Cubic>, {arriveTangent=-2.1279, tangentMode=RCTM_User, time=1, value=0}]}}}
     Scale Curve=&RandomRangeFloat(Minimum=-2, Maximum=2)
    ))
   &ScaleSpriteSize(Uniform Curve Sprite Scale=DataInterface() {propertyValues {Curve {keys [
        {interpMode=RCIM_Cubic, tangentMode=RCTM_User, time=0, value=0.8}
        {interpMode=RCIM_Cubic, tangentMode=RCTM_User, time=0.5, value=1.5}
        {tangentMode=RCTM_User, time=1, value=0}
       ]}}})
   &Drag(Rotational Drag=0.1, Use Linear Drag=true, Use Rotational Drag=true)
   &SolveForcesAndVelocity(Clamp Velocity=false, Limit Acceleration=false)
  ]
  rendererClasses [&NiagaraSpriteRendererProperties]
  renderers [
   {rendererClass=&NiagaraSpriteRendererProperties}
   {rendererClass=&NiagaraSpriteRendererProperties, rendererIndex=1}
  ]
 }
 Dust {
  emitterUpdateScript [
   &EmitterState()
   &SpawnRate(SpawnRate=12, Use Spawn Probability=false)
  ]
  particleSpawnScript [
   &PartitionParticles(Particles Per Partition=4, Partition Mode:ENiagara_PartitionMode=::NewEnumerator1)
   &InitializeParticle(
    Color=&SelectLinearColorFromArray(
     Array Sampling Mode=::NewEnumerator0
     Direct Array Index=?
     Color Selection Array=DataInterface() {propertyValues {ColorData [<rgba 1 0 0.887 1>, <rgba 1 0.04 0.2 1>, <rgba 0.75 0.2974 0 1>, <rgba 0 0.5 0.5 1>, <rgba 0 0.1448 2 1>, <rgba 0.1764 0.04 2 1>]}}
     Random Seed=0
    )
    Sprite Rotation Angle=&RandomRangeFloat(Minimum=0.2, Maximum=0.4)
    Sprite Rotation Mode=::NewEnumerator1
    Sprite Size Mode=::NewEnumerator3
    Uniform Sprite Size=&Multiply_Float(A=20, B=&MakeCustomFloatFromBool(Bool=$User.IsLargeBody, True Float=1.5, False Float=1))
    UsePositionOffset=false
    Write Material Random=true
   )
  ]
  particleUpdateScript [
   &ParticleState()
   &SpriteRotationRate(Rotation Rate=30) {enabled=false}
   &SkeletalMeshLocation(
    Bone / Socket Index=&SelectIntFromArray(Direct Array Index=$Particles.PartitionParticles.Partition, Int Selection Array=DataInterface() {propertyValues {IntData [1,2]}})
    Bone Sampling Mode=::NewEnumerator4
    MaskSpawnGroup=false
    Orientation Sampling=::NewEnumerator1
    Position Sampling=::NewEnumerator1
    Sampled Position Offset=&Lerp_Vector(
     A=&SelectVectorFromArray(
      Direct Array Index=$Particles.PartitionParticles.Partition
      Vector Selection Array=DataInterface() {propertyValues {FloatData [<-5 0 3>, <5 0 -3>]}}
     )
     B=&SelectVectorFromArray(
      Direct Array Index=$Particles.PartitionParticles.Partition
      Vector Selection Array=DataInterface() {propertyValues {FloatData [<27 0 3>, <-27 0 -3>]}}
     )
     Alpha=$Particles.PartitionParticles.Normalized Index In Partition
    )
    Skeletal Mesh=$System.Skeletal Mesh
   )
   &DynamicMaterialParameters(
    Index 0 Param 1=0.5
    Module.Index0Float0Write=true
    Module.Index0Float1Write=false
    Module.Index0Float2Write=false
    Module.Index0Float3Write=false
   )
   &ScaleColor(
    Scale Alpha=&FloatFromCurve(FloatCurve=DataInterface() {propertyValues {Curve {keys [
         {interpMode=RCIM_Cubic, leaveTangent=2.1138, tangentMode=RCTM_Break, time=0, value=0}
         <at 0.4 0.95 RCIM_Cubic>
         {arriveTangent=-3.4143, interpMode=RCIM_Cubic, leaveTangent=-3.4143, tangentMode=RCTM_User, time=1, value=0}
        ]}}})
    Scale RGB=&VectorFromFloat(Value=&MakeCustomFloatFromBool(Bool=$User.IsFrontEnd, True Float=1, False Float=3))
    ScaleA=true
    ScaleRGB=true
   )
   &GravityForce(Gravity=&RandomRangeVector(Minimum=<0 0 15>, Maximum=<0 0 25>))
   &SolveForcesAndVelocity(Clamp Velocity=false, Limit Acceleration=false)
  ]
  rendererClasses [&NiagaraSpriteRendererProperties]
  renderers [{rendererClass=&NiagaraSpriteRendererProperties}]
 }
 FigureEightBeams {
  emitterSpawnScript [/SetVariables(
    Emitter.CenterBrightness:NiagaraFloat=&MakeCustomFloatFromBool(Bool=$User.IsFrontEnd, True Float=1, False Float=0.5)
    Emitter.CoronaBrightness:NiagaraFloat=&MakeCustomFloatFromBool(Bool=$User.IsFrontEnd, True Float=0.125, False Float=0.2)
   ) {isSetParameters=true}]
  emitterUpdateScript [
   &EmitterState()
   &SpawnBurst_Instantaneous(Use Spawn Probability=false, UseLoopCountLimit=true)
   &SpawnBurst_Instantaneous(Spawn Group=1, Use Spawn Probability=false, UseLoopCountLimit=true) {module=SpawnBurst_Instantaneous001}
  ]
  particleSpawnScript [&InitializeParticle(
    Position Offset=&SelectVectorFromArray(
     Direct Array Index=&SpawnGroup()
     Vector Selection Array=DataInterface() {propertyValues {FloatData [<40 -35 150>, <-40 -35 150>]}}
    )
    Position Offset Coordinate Space=::Local
    Sprite Size=<75 200>
    Sprite Size Mode=::NewEnumerator4
    UsePositionOffset=true
    Write Material Random=true
   )]
  particleUpdateScript [
   &ParticleState(Kill Particles When Lifetime Has Elapsed=false, Let Infinitely Lived Particles Die When Emitter Deactivates=true)
   /SetVariables(
    Particles.PositionTargetA:NiagaraPosition=&Add_Vector(
     A=<0 0 1>
     B=&self:DrawFigureEight(
      Length Scale:NiagaraFloat=100
      Height Scale:NiagaraFloat=100
      Time:NiagaraFloat=&Add_Float(
       A=&MaskFloatBySpawnGroup(Masked Float=0.25, Passthrough Float=0)
       B=&Add_Float(A=$System.SystemRandom, B=&Multiply_Float(A=0.2, B=$Engine.Time))
      )
     )
    )
    Particles.UniqueAge:NiagaraFloat=&Add_Float(A=$Emitter.Age, B=&RandomRangeFloat(Minimum=0, Maximum=1))
   ) {isSetParameters=true}
   &SpriteFacingAndAlignment(
    Sprite Alignment=&Subtract_Position(A=$Particles.Position, B=$Particles.PositionTargetA)
    WriteAlignment=true
    WriteFacing=false
   )
   /SetVariables(Transient.NormlizedBeamLength:NiagaraFloat=&RemapRange(
     Input Value To Remap=&DistanceBetweenPositions(A=$Particles.Position, B=$Particles.PositionTargetA)
     Input Min=150
     Input Max=210
     Output Min=0
     Output Max=1
    )) {isSetParameters=true}
   /SetVariables(Transient.NormlizedBeamLength:NiagaraFloat=&RemapRange(
     Input Value To Remap=&DistanceBetweenPositions(A=$Particles.Position, B=$Particles.PositionTargetA)
     Input Min=&MaskFloatBySpawnGroup(Masked Float=140, Passthrough Float=165)
     Input Max=&MaskFloatBySpawnGroup(Masked Float=208, Passthrough Float=233)
     Output Min=0
     Output Max=1
    )) {enabled=false, isSetParameters=true}
   &ScaleColor(
    Scale Alpha=&Add_Float(
     A=&MakeCustomFloatFromBool(Bool=$User.IsFrontEnd, True Float=0, False Float=0.25)
     B=&FloatFromCurve(
      FloatCurve=DataInterface() {propertyValues {Curve {keys [
          {interpMode=RCIM_Cubic, tangentMode=RCTM_User, time=0, value=0.3}
          {interpMode=RCIM_Cubic, tangentMode=RCTM_User, time=0.5, value=1}
          {tangentMode=RCTM_User, time=1, value=0.3}
         ]}}}
      CurveIndex=&Frac_Float(Float=$Particles.UniqueAge)
      Scale Curve=0.5
     )
    )
    Scale RGB=&Multiply_VectorByFloat(
     Vector=&MakeVectorFromLinearColorRGB(Color=&ColorFromCurve(
       ColorCurve=DataInterface() {propertyValues {
         AlphaCurve {keys [<at 0 1 RCIM_Cubic>]}
         BlueCurve {keys [<at 0 0.8313 RCIM_Cubic>,<at 0.5645 0.793>,<at 1 0.2 RCIM_Cubic>]}
         GreenCurve {keys [<at 0 0.0166 RCIM_Cubic>,<at 0.5645 0.0574>,<at 1 0.2 RCIM_Cubic>]}
         RedCurve {keys [<at 0 0.0733 RCIM_Cubic>,<at 0.5645 0>,<at 1 0 RCIM_Cubic>]}
        }}
       CurveIndex=&RandomRangeFloat(
        Minimum=0
        Maximum=1
        Evaluation Type=::NewEnumerator1
        Randomness Mode=::NewEnumerator1
        Random Seed=&MakeIntFromFloat(Float=$Particles.UniqueAge)
        Fixed Random Seed=true
       )
      ))
     Float=&MakeCustomFloatFromBool(Bool=$User.IsFrontEnd, True Float=1, False Float=1)
    )
    ScaleA=true
    ScaleRGB=true
   )
   /SetVariables(
    Particles.GroundSpriteColor:LinearColor=&MultiplyLinearColor(
     Linear Color=$Particles.Color
     Scale Factor=&MakeFloat4FromVectorAndFloat(
      "Vector (RGB)"=&VectorFromFloat(Value=&Lerp_Float(A=1, B=0.25, Alpha=$Transient.NormlizedBeamLength))
      "Float (Alpha)"=&FloatFromCurve(
       FloatCurve=DataInterface() {propertyValues {Curve {keys [
           {interpMode=RCIM_Cubic, tangentMode=RCTM_User, time=0, value=0.5}
           {interpMode=RCIM_Cubic, tangentMode=RCTM_User, time=0.5, value=1}
           {tangentMode=RCTM_User, time=1, value=0.5}
          ]}}}
       CurveIndex=&Frac_Float(Float=$Particles.UniqueAge)
       Scale Curve=&MakeCustomFloatFromBool(Bool=$User.IsFrontEnd, True Float=1, False Float=0.5)
      )
     )
    )
    Particles.GroundSpriteFacing:Vector3f=<0 0 1>
    Particles.GroundSpriteScale:Vector2f=&Vector2DFromCurve(
     Vector2Curve=DataInterface() {propertyValues {XCurve {keys [<at 0 1>]}, YCurve {keys [<at 0 1>,<at 1 1.5>]}}}
     CurveIndex=$Transient.NormlizedBeamLength
     ScaleCurve=&Vector2DFromFloat(Value=40)
    )
   ) {isSetParameters=true}
   /SetVariables(
    Particles.GroundSpriteFacing:Vector3f=<0 0 1>
    Particles.GroundSpriteScale:Vector2f=&Vector2DFromFloat(Value=80)
   ) {enabled=false, isSetParameters=true}
   /SetVariables(
    Particles.SourceSpriteColor:LinearColor=&MultiplyLinearColor(
     Linear Color=&LerpLinearColors(StartColor=$Particles.Color, EndColor=<rgba 1 1 1 1>, LerpFactor=0.75)
     Scale Factor=&MakeFloat4FromVectorAndFloat(
      "Vector (RGB)"=&VectorFromFloat(Value=&MakeCustomFloatFromBool(Bool=$User.IsFrontEnd, True Float=0.5, False Float=1))
      "Float (Alpha)"=&MakeCustomFloatFromBool(Bool=$User.IsFrontEnd, True Float=0.75, False Float=1)
     )
    )
    Particles.SourceSpriteScale:Vector2f=&Vector2DFromFloat(Value=50)
   ) {isSetParameters=true}
   /SetVariables(
    Particles.SourceGlowColor:LinearColor=&MultiplyLinearColor(
     Linear Color=&LerpLinearColors(
      StartColor=$Particles.Color
      EndColor=<rgba 1 1 1 1>
      LerpFactor=&MakeCustomFloatFromBool(Bool=$User.IsFrontEnd, True Float=0.25, False Float=0)
     )
     Scale Factor=&MakeFloat4FromVectorAndFloat(
      "Vector (RGB)"=&VectorFromFloat(Value=&MakeCustomFloatFromBool(Bool=$User.IsFrontEnd, True Float=2, False Float=2))
      "Float (Alpha)"=&MakeCustomFloatFromBool(Bool=$User.IsFrontEnd, True Float=1, False Float=1)
     )
    )
    Particles.SourceGlowScale:Vector2f=&Vector2DFromFloat(Value=15)
   ) {isSetParameters=true}
   /SetVariables(Particles.BackerColor:LinearColor=<rgba 0.028 0.0305 0.1287 0.5>) {enabled=false, isSetParameters=true}
   &ScaleColor(
    Scale Alpha=&FloatFromCurve(
     FloatCurve=DataInterface() {propertyValues {Curve {keys [
         {interpMode=RCIM_Cubic, tangentMode=RCTM_User, time=0, value=0.3}
         {interpMode=RCIM_Cubic, tangentMode=RCTM_User, time=0.5, value=1}
         {tangentMode=RCTM_User, time=1, value=0.3}
        ]}}}
     CurveIndex=&Frac_Float(Float=$Particles.UniqueAge)
     Scale Curve=0.5
    )
    Scale RGB=&MakeCustomVectorFromBool(
     Bool=$User.IsFrontEnd
     True Vector=&MakeVectorFromLinearColorRGB(Color=&ColorFromCurve(
       ColorCurve=DataInterface() {propertyValues {
         AlphaCurve {keys [<at 0 1 RCIM_Cubic>]}
         BlueCurve {keys [
           <at 0 0.4191>
           <at 0.163 0.5344>
           {arriveTangent=0.5563, interpMode=RCIM_Cubic, leaveTangent=0.5563, time=0.5291, value=0.4781}
           <at 1 1 RCIM_Cubic>
          ]}
         GreenCurve {keys [
           <at 0 0.6972>
           <at 0.163 1>
           {arriveTangent=-0.9784, interpMode=RCIM_Cubic, leaveTangent=-0.9784, time=0.5291, value=0.4781}
           <at 1 0.1811 RCIM_Cubic>
          ]}
         RedCurve {keys [
           <at 0 1>
           <at 0.163 0.9714>
           {arriveTangent=-1.1606, interpMode=RCIM_Cubic, leaveTangent=-1.1606, time=0.5291, value=0.4781}
           <at 1 0 RCIM_Cubic>
          ]}
        }}
       CurveIndex=&RandomRangeFloat(
        Minimum=0
        Maximum=1
        Evaluation Type=::NewEnumerator1
        Randomness Mode=::NewEnumerator1
        Random Seed=&Add_Integer(
         A=&MakeIntFromFloat(Float=$System.SystemRandom)
         B=&MakeIntFromFloat(Float=$Particles.UniqueAge)
        )
        Fixed Random Seed=true
       )
      ))
     False Vector=&self:AbsVectorChannels(Input:Vector3f=&Add_Vector(
       A=&VectorFromFloat(Value=0.25)
       B=&RandomVector(
        Vector Scale=0.75
        Evaluation Type=::NewEnumerator1
        Randomness Mode=::NewEnumerator1
        Random Seed=&Add_Integer(
         A=&SpawnGroup()
         B=&MakeIntFromFloat(Float=&Add_Float(A=$System.SystemRandom, B=$Particles.UniqueAge))
        )
        Fixed Random Seed=true
       )
      ))
    )
    ScaleA=true
    ScaleRGB=true
   ) {enabled=false, module=ScaleColor001}
   &DynamicMaterialParameters(
    Index 0 Param 1=0.9
    Index 0 Param 2=&FloatFromCurve(
     FloatCurve=DataInterface() {propertyValues {Curve {keys [
         {interpMode=RCIM_Cubic, tangentMode=RCTM_User, time=0, value=0.95}
         {interpMode=RCIM_Cubic, tangentMode=RCTM_User, time=0.5, value=1}
         {tangentMode=RCTM_User, time=1, value=0.95}
        ]}}}
     CurveIndex=&Frac_Float(Float=&Add_Float(A=$Engine.System.Age, B=&RandomRangeFloat(Minimum=0, Maximum=1)))
    )
    Index 1 Param 1=2
    Index 2 Param 1=1
    Module.Index0Float0Write=true
    Module.Index0Float1Write=true
    Module.Index0Float2Write=true
    Module.Index0Float3Write=true
    Module.Index1Float0Write=true
    Module.Index1Float1Write=true
    Module.Index1Float2Write=true
    Module.Index1Float3Write=true
    Module.Index2Float0Write=true
    Module.Index2Float1Write=false
    Module.Index2Float2Write=false
    Module.Index2Float3Write=false
    Write Parameter Index 1=::NewEnumerator1
    Write Parameter Index 2=::NewEnumerator1
   )
   &Color(Color=&ColorFromCurve(
     ColorCurve=DataInterface() {propertyValues {
       AlphaCurve {keys [<at 0 1 RCIM_Cubic>]}
       BlueCurve {keys [<at 0 1>,<at 0.02 0 RCIM_Cubic>,<at 0.98 1 RCIM_Cubic>,<at 1 0>]}
       GreenCurve {keys [
         <at 0 1>
         {arriveTangent=-1.0204, interpMode=RCIM_Cubic, leaveTangent=-1.0204, time=0.02, value=0}
         {arriveTangent=1.0204, interpMode=RCIM_Cubic, leaveTangent=1.0204, time=0.98, value=0}
         <at 1 1>
        ]}
       RedCurve {keys [
         <at 0 1>
         {arriveTangent=-1.0204, interpMode=RCIM_Cubic, leaveTangent=-1.0204, time=0.02, value=1}
         {arriveTangent=-1.0204, interpMode=RCIM_Cubic, leaveTangent=-1.0204, time=0.98, value=0}
         <at 1 0>
        ]}
      }}
     CurveIndex=$Transient.NormlizedBeamLength
    )) {enabled=false}
  ]
  rendererClasses [&NiagaraSpriteRendererProperties]
  renderers [
   {rendererClass=&NiagaraSpriteRendererProperties}
   {rendererClass=&NiagaraSpriteRendererProperties, rendererIndex=1}
   {rendererClass=&NiagaraSpriteRendererProperties, rendererIndex=2}
   {rendererClass=&NiagaraSpriteRendererProperties, rendererIndex=3}
   {rendererClass=&NiagaraSpriteRendererProperties, rendererIndex=4}
  ]
 }
 Glows {
  emitterUpdateScript [
   &EmitterState()
   &SpawnBurst_Instantaneous(Spawn Count=2, Use Spawn Probability=false, UseLoopCountLimit=true)
   &SpawnBurst_Instantaneous(Spawn Count=10, Spawn Group=1, Use Spawn Probability=false, UseLoopCountLimit=true) {module=SpawnBurst_Instantaneous001}
   &SpawnBurst_Instantaneous(Spawn Count=7, Spawn Group=2, Use Spawn Probability=false, UseLoopCountLimit=true) {module=SpawnBurst_Instantaneous002}
  ]
  particleSpawnScript [
   /SetVariables(
    Particles.PickBone:NiagaraInt32=&MaskIntBySpawnGroup(
     Masked Int=7
     Passthrough Int=&MaskIntBySpawnGroup(
      Masked Int=&SelectIntFromArray(
       Direct Array Index=&ReturnExecIndex()
       Int Selection Array=DataInterface() {propertyValues {IntData [1,1,2,2,3,3,4,4,5,6]}}
      )
      Passthrough Int=&ReturnExecIndex()
      Spawn Group Mask=1
     )
     Spawn Group Mask=0
    )
    Particles.PickBoneMatch:NiagaraInt32=&SelectIntFromArray(
     Direct Array Index=&ReturnExecIndex()
     Int Selection Array=DataInterface() {propertyValues {IntData [1,3,2,4,3,5,4,6,5,6]}}
    )
    Particles.ScaleGroup:NiagaraInt32=&MaskIntBySpawnGroup(
     Masked Int=&SelectIntFromArray(
      Direct Array Index=&ReturnExecIndex()
      Int Selection Array=DataInterface() {propertyValues {IntData [0,1,0,1,2,3,2,3,4,4]}}
     )
     Passthrough Int=&SelectIntFromArray(
      Direct Array Index=&ReturnExecIndex()
      Int Selection Array=DataInterface() {propertyValues {IntData [0,1,1,2,2,3,3]}}
     )
     Spawn Group Mask=1
    )
    Particles.VisibilityTag=&SpawnGroup()
   ) {isSetParameters=true}
   /SetVariables(
    Transient.BodyScale:NiagaraFloat=&SelectFloatFromArray(Direct Array Index=$Particles.ScaleGroup, Float Selection Array=DataInterface() {propertyValues {FloatData [35,30,35,30,25]}})
    Transient.DecalScale:NiagaraFloat=&SelectFloatFromArray(Direct Array Index=$Particles.ScaleGroup, Float Selection Array=DataInterface() {propertyValues {FloatData [40,40,35,30]}})
   ) {isSetParameters=true}
   &InitializeParticle(
    Lifetime=2
    Position Offset=<0 -30 110>
    Sprite Rotation Mode=::NewEnumerator2
    Sprite Size Mode=::NewEnumerator3
    Uniform Sprite Size=&Multiply_Float(
     A=&MaskFloatBySpawnGroup(
      Masked Float=100
      Passthrough Float=&MaskFloatBySpawnGroup(Masked Float=$Transient.BodyScale, Passthrough Float=$Transient.DecalScale, Spawn Group Mask=1)
     )
     B=&MakeCustomFloatFromBool(Bool=$User.IsLargeBody, True Float=1.35, False Float=1)
    )
    UsePositionOffset=false
    Write Material Random=true
   )
  ]
  particleUpdateScript [
   &ParticleState(Kill Particles When Lifetime Has Elapsed=false, Let Infinitely Lived Particles Die When Emitter Deactivates=true)
   &SpriteRotationRate(Rotation Rate=&MaskFloatBySpawnGroup(
     Masked Float=&SelectFloatFromArray(
      Direct Array Index=&ReturnExecIndex()
      Float Selection Array=DataInterface() {propertyValues {FloatData [-15,20]}}
     )
     Passthrough Float=0
    ))
   /SetVariables(
    Transient.MainColorAverage:Vector3f=<1 0.2 0.444>
    Transient.ToDMult:NiagaraFloat=&FNTimeOfDayScalar(
     Night Scale=1.5
     Morning Scale=1.2
     Day Scale=&MakeCustomFloatFromBool(Bool=$User.IsFrontEnd, True Float=1.2, False Float=1.65)
     Evening Scale=0.9
    )
   ) {isSetParameters=true}
   /SetVariables(
    Particles.BodyColor:LinearColor=&MakeLinearColorFromVectorAndFloat(
     "Vector (RGB)"=&Lerp_Vector(A=$System.MainColor, B=$Transient.MainColorAverage, Alpha=0.6)
     "Float (Alpha)"=&Multiply_Float(
      A=&Multiply_Float(A=$Transient.ToDMult, B=&MakeCustomFloatFromBool(Bool=$User.IsFrontEnd, True Float=0.8, False Float=1))
      B=&Lerp_Float(A=0.25, B=0.1, Alpha=&Multiply_FloatByInt(Float=0.25, Integer=$Particles.ScaleGroup))
     )
    )
    Particles.ChestColor:LinearColor=&ColorFromCurve(
     ColorCurve=DataInterface() {propertyValues {
       AlphaCurve {keys [
         {interpMode=RCIM_Cubic, leaveTangentWeight=0.3263, tangentMode=RCTM_User, time=0, value=2}
         {
          arriveTangent=-1.6951
          arriveTangentWeight=0.0481
          interpMode=RCIM_Cubic
          leaveTangent=-1.6951
          leaveTangentWeight=0.0593
          tangentMode=RCTM_User
          time=0.078
          value=1
         }
         {
          arriveTangent=-0.595
          arriveTangentWeight=0.0949
          interpMode=RCIM_Cubic
          leaveTangent=-0.595
          leaveTangentWeight=0.0915
          tangentMode=RCTM_User
          time=0.262
          value=0.6
         }
         {
          arriveTangent=-0.2363
          arriveTangentWeight=0.099
          interpMode=RCIM_Cubic
          leaveTangent=-0.2363
          leaveTangentWeight=0.1203
          tangentMode=RCTM_User
          time=0.469
          value=0.5
         }
         {
          arriveTangent=0.6689
          arriveTangentWeight=0.1073
          interpMode=RCIM_Cubic
          leaveTangent=0.6689
          leaveTangentWeight=0.104
          tangentMode=RCTM_User
          time=0.667
          value=0.4
         }
         {
          arriveTangent=2.8194
          arriveTangentWeight=0.357
          interpMode=RCIM_Cubic
          leaveTangent=2.8194
          leaveTangentWeight=0.2529
          tangentMode=RCTM_User
          time=0.859
          value=1
         }
         {arriveTangentWeight=0.0453, interpMode=RCIM_Cubic, tangentMode=RCTM_User, time=0.995, value=2}
        ]}
       BlueCurve {keys [<at 0 1>, {arriveTangent=-1.5, interpMode=RCIM_Cubic, leaveTangent=-1.5, time=0.3333, value=0}, <at 0.6667 0>, <at 1 1 RCIM_Cubic>]}
       GreenCurve {keys [<at 0 0.15>, {arriveTangent=1.275, interpMode=RCIM_Cubic, leaveTangent=1.275, time=0.3333, value=0}, <at 0.6667 1>, <at 1 0.15 RCIM_Cubic>]}
       RedCurve {keys [<at 0 0.15>, {arriveTangent=-0.225, interpMode=RCIM_Cubic, leaveTangent=-0.225, time=0.3333, value=1}, <at 0.6667 0>, <at 1 0.15 RCIM_Cubic>]}
      }}
     CurveIndex=&Frac_Float(Float=&Add_Float(
       A=&Multiply_Float(A=0.4, B=$System.Age)
       B=&Multiply_FloatByInt(Float=0.2, Integer=&ReturnExecIndex())
      ))
    )
    Particles.DecalColor:LinearColor=&MakeLinearColorFromVectorAndFloat(
     "Vector (RGB)"=&Multiply_VectorByFloat(
      Vector=&Lerp_Vector(A=$System.MainColor, B=$Transient.MainColorAverage, Alpha=0.25)
      Float=&Multiply_Float(
       A=$Transient.ToDMult
       B=&Lerp_Float(A=3, B=1, Alpha=&Multiply_FloatByInt(Float=0.3333, Integer=$Particles.ScaleGroup))
      )
     )
     "Float (Alpha)"=&MakeCustomFloatFromBool(Bool=$User.IsLargeBody, True Float=0.25, False Float=1)
    )
   ) {isSetParameters=true}
   /SetVariables(Transient.Linear Color:LinearColor=&ColorFromCurve(
     ColorCurve=DataInterface() {propertyValues {
       AlphaCurve {keys [<at 0 1 RCIM_Cubic>]}
       BlueCurve {keys [
         <at 0 3>
         <at 0.1658 2>
         <at 0.1893 0.3511>
         {arriveTangent=-1.3088, interpMode=RCIM_Cubic, leaveTangent=-1.3088, time=0.3333, value=0}
         <at 0.4712 -0.018>
         <at 0.5052 -0.0079>
         <at 0.6667 0>
         <at 0.8237 0.2052>
         <at 0.8629 1.5>
         <at 1 3 RCIM_Cubic>
        ]}
       GreenCurve {keys [
         <at 0 0>
         <at 0.1658 0>
         <at 0.1893 0>
         {arriveTangent=0.7099, interpMode=RCIM_Cubic, leaveTangent=0.7099, time=0.3333, value=0}
         <at 0.4712 0.2002>
         <at 0.5052 0.5>
         <at 0.6667 0.75>
         <at 0.8237 0.5>
         <at 0.8629 0.6011>
         <at 1 0 RCIM_Cubic>
        ]}
       RedCurve {keys [
         <at 0 0>
         <at 0.1658 0.6829>
         <at 0.1893 1>
         {arriveTangent=-1.7733, interpMode=RCIM_Cubic, leaveTangent=-1.7733, time=0.3333, value=1}
         <at 0.4712 0.5>
         <at 0.5052 0.2211>
         <at 0.6667 0>
         <at 0.8237 0>
         <at 0.8629 0>
         <at 1 0 RCIM_Cubic>
        ]}
      }}
     CurveIndex=&Frac_Float(Float=&Add_Float(
       A=&Multiply_Float(A=$Engine.System.Age, B=&RandomRangeFloat(Minimum=0.2, Maximum=0.2))
       B=&RandomRangeFloat(Minimum=0, Maximum=1)
      ))
    )) {enabled=false, isSetParameters=true}
   /SetVariables(
    Particles.BodyColor:LinearColor=&MultiplyLinearColor(
     Linear Color=&MakeCustomLinearColorFromBool(Bool=$User.IsFrontend, True Color=<rgba 2 2 2 0.1>, False Color=<rgba 1 1 1 0.5>)
     Scale Factor=&MakeVector4_FromLinearColor(LinearColor=&ColorFromCurve(
       ColorCurve=DataInterface() {propertyValues {
         AlphaCurve {keys [<at 0 1>,<at 0.4671 0.6>,<at 1 1>]}
         BlueCurve {keys [<at 0.1217 0.8714>,<at 0.3306 1>,<at 0.4729 0.7292>,<at 0.6336 0.8714>,<at 0.8205 0.8714>]}
         GreenCurve {keys [<at 0.1217 0.3515>,<at 0.3306 0.3241>,<at 0.4729 0.2521>,<at 0.6336 0.1894>,<at 0.8205 0.3515>]}
         RedCurve {keys [<at 0.1217 0.227>,<at 0.3306 0.3406>,<at 0.4729 0.1822>,<at 0.6336 0.2246>,<at 0.8205 0.227>]}
        }}
       CurveIndex=$Transient.CurveOffset
      ))
    )
    Particles.HeadColor:LinearColor=&MultiplyLinearColor(
     Linear Color=&HueShiftLinearColor(
      Hue=&RandomRangeFloat(Minimum=-0.1, Maximum=0.1)
      Saturation=1
      Value=1
      Linear Color=&MakeCustomLinearColorFromBool(Bool=$User.IsFrontend, True Color=<rgba 0.598 0.6062 1 0.25>, False Color=<rgba 0.598 0.6062 1 0.75>)
     )
     Scale Factor=&MakeVector4_FromLinearColor(LinearColor=&ColorFromCurve(
       ColorCurve=DataInterface() {propertyValues {
         AlphaCurve {keys [<at 0 1>,<at 0.4671 0.6>,<at 1 1>]}
         BlueCurve {keys [<at 0.1217 0.7605>,<at 0.3306 1>,<at 0.4729 0.7292>,<at 0.6336 0.8714>,<at 0.8205 0.7605>]}
         GreenCurve {keys [<at 0.1217 0.6308>,<at 0.3306 0.9701>,<at 0.4729 0.4584>,<at 0.6336 0.5208>,<at 0.8205 0.6308>]}
         RedCurve {keys [<at 0.1217 0.4287>,<at 0.3306 0.7067>,<at 0.4729 0.3955>,<at 0.6336 0.5719>,<at 0.8205 0.4287>]}
        }}
       CurveIndex=$Transient.CurveOffset
      ))
    )
   ) {enabled=false, isSetParameters=true}
   &DynamicMaterialParameters(
    Index 0 Param 1=1
    Index 0 Param 2=&RandomRangeFloat(Minimum=0, Maximum=0.2)
    Index 1 Param 1=2
    Index 2 Param 1=&SelectFloatFromArray(
     Direct Array Index=&ReturnExecIndex()
     Float Selection Array=DataInterface() {propertyValues {FloatData [5,5,5,5,3,3]}}
    )
    Module.Index0Param0Write=true
    Module.Index0Param1Write=true
    Module.Index0Param2Write=false
    Module.Index0Param3Write=false
    Module.Index1Param0Write=true
    Module.Index1Param1Write=true
    Module.Index1Param2Write=false
    Module.Index1Param3Write=false
    Module.Index2Param0Write=true
    Module.Index2Param1Write=true
    Module.Index2Param2Write=false
    Module.Index2Param3Write=false
    Write Parameter Index 1=true
    Write Parameter Index 2=true
   )
   &SkeletalMeshLocation(
    Bone / Socket Index=$Particles.PickBone
    Bone Sampling Mode=::NewEnumerator4
    MaskSpawnGroup=false
    Orientation Sampling=::NewEnumerator1
    Skeletal Mesh=$System.Skeletal Mesh
   )
   &SkeletalMeshLocation(
    Bone / Socket Index=$Particles.PickBoneMatch
    Bone Sampling Mode=::NewEnumerator4
    MaskSpawnGroup=false
    Orientation Sampling=::NewEnumerator1
    Position Sampling=::NewEnumerator2
    Skeletal Mesh=$System.Skeletal Mesh
   ) {module=SkeletalMeshLocation001}
   &CameraOffset(Camera Offset Amount=&SelectFloatFromArray(
     Direct Array Index=&SpawnGroup()
     Float Selection Array=DataInterface() {propertyValues {FloatData [-40,-20,5]}}
    ))
   /SetVariables(Particles.GlowPosition:NiagaraPosition=&LerpPosition(A=$Particles.SkeletalMeshLocation.SampledPosition, B=$Particles.SkeletalMeshLocation001.SampledPosition, Alpha=0.5)) {isSetParameters=true}
  ]
  rendererClasses [&NiagaraSpriteRendererProperties]
  renderers [
   {rendererClass=&NiagaraSpriteRendererProperties}
   {rendererClass=&NiagaraSpriteRendererProperties, rendererIndex=1}
   {rendererClass=&NiagaraSpriteRendererProperties, rendererIndex=2}
  ]
 }
 MainLight {
  emitterSpawnScript [/SetVariables(
    Emitter.CenterBrightness:NiagaraFloat=&MakeCustomFloatFromBool(Bool=$User.IsFrontEnd, True Float=2, False Float=10)
    Emitter.CoronaBrightness:NiagaraFloat=&MakeCustomFloatFromBool(Bool=$User.IsFrontEnd, True Float=0.125, False Float=0.2)
   ) {isSetParameters=true}]
  emitterUpdateScript [
   &EmitterState()
   &SpawnBurst_Instantaneous(Use Spawn Probability=false, UseLoopCountLimit=true)
  ]
  particleSpawnScript [&InitializeParticle(
    Lifetime=5
    Position Offset=<30 -30 200>
    Position Offset Coordinate Space=::Local
    Sprite Size=<150 300>
    Sprite Size Mode=::NewEnumerator4
    UsePositionOffset=true
    Write Material Random=true
   )]
  particleUpdateScript [
   &ParticleState(Kill Particles When Lifetime Has Elapsed=false, Let Infinitely Lived Particles Die When Emitter Deactivates=true)
   &SampleSkeletalMesh(
    AdditionalYawPitchRoll=false
    Bone / Socket Index=7
    Bone Sampling Mode:ENiagaraBoneSamplingMode=::NewEnumerator4
    Sampled Velocity Coordinate Space:ENiagaraCoordinateSpace=::World
    Skeletal Mesh:NiagaraDataInterfaceSkeletalMesh=$System.Skeletal Mesh
   )
   &InterpolateOverTime(
    Initial Position Value:NiagaraPosition=$Particles.SampleSkeletalMesh.SampledPosition
    Target Position Value:NiagaraPosition=$Particles.SampleSkeletalMesh.SampledPosition
    Variable Type=::NewEnumerator6
   )
   &SpriteFacingAndAlignment(
    Sprite Alignment=&Subtract_Position(A=$Particles.Position, B=$Particles.InterpolateOverTime.InterpolateOverTime_Position.Moving Average)
    WriteAlignment=true
    WriteFacing=false
   )
   /SetVariables(
    Particles.GroundSpriteFacing:Vector3f=&Lerp_Vector(
     A=&NormalizeVector(Vector To Normalize=$Particles.SpriteAlignment, Safe Normalize=true, Safe Fallback Threshold=0, Safe Fallback Vector=<0 0 0.001>)
     B=&RotateVectorByQuaternion(Vector To Rotate=<0 1 0>, Quaternion=$Particles.SampleSkeletalMesh.SampledRotation)
     Alpha=0.75
    )
    Particles.GroundSpritePosition:NiagaraPosition=&AddVectorToPosition(
     Position=$Particles.InterpolateOverTime.InterpolateOverTime_Position.Moving Average
     Vector=&Multiply_VectorByFloat(
      Vector=&NormalizeVector(Vector To Normalize=$Particles.SpriteAlignment, Safe Normalize=true, Safe Fallback Threshold=0, Safe Fallback Vector=<0 0 0.001>)
      Float=45
     )
    )
    Particles.GroundSpriteScale:Vector2f=&Vector2DFromFloat(Value=80)
   ) {enabled=false, isSetParameters=true}
   &ScaleColor(Scale RGB=$System.MainColor, ScaleA=false, ScaleRGB=true)
   &Color(Color=&HueShiftLinearColor(
     Hue=&RandomRangeFloat(
      Minimum=0
      Maximum=1
      Evaluation Type=::NewEnumerator1
      Randomness Mode=::NewEnumerator1
      Random Seed=&MakeIntFromFloat(Float=&Add_Float(A=$Emitter.Age, B=$System.SystemRandom))
      Fixed Random Seed=true
     )
     Saturation=1
     Value=1
     Linear Color=$Particles.Initial.Color
    )) {enabled=false}
   /SetVariables(
    Particles.LightColor=&Multiply_LinearColorByFloat(Linear Color=$Particles.Color, Float=30)
    Particles.LightPosition:NiagaraPosition=&AddVectorToPosition(
     Position=$Particles.InterpolateOverTime.InterpolateOverTime_Position.Moving Average
     Vector=&Multiply_VectorByFloat(
      Vector=&NormalizeVector(Vector To Normalize=$Particles.SpriteAlignment, Safe Normalize=true, Safe Fallback Threshold=0, Safe Fallback Vector=<0 0 0.001>)
      Float=&MakeCustomFloatFromBool(Bool=$User.IsLargeBody, True Float=60, False Float=50)
     )
    )
    Particles.LightRadius=50
   ) {isSetParameters=true}
   /SetVariables(
    Particles.SourceSpriteColor:LinearColor=&MultiplyLinearColor(
     Linear Color=&LerpLinearColors(StartColor=$Particles.Color, EndColor=<rgba 1 1 1 1>, LerpFactor=0.25)
     Scale Factor=&MakeFloat4FromVectorAndFloat(
      "Vector (RGB)"=&VectorFromFloat(Value=&MakeCustomFloatFromBool(Bool=$User.IsFrontEnd, True Float=0.5, False Float=2))
      "Float (Alpha)"=0.5
     )
    )
    Particles.SourceSpriteScale:Vector2f=&Vector2DFromFloat(Value=100)
   ) {isSetParameters=true}
   /SetVariables(
    Particles.SourceGlowColor:LinearColor=&MultiplyLinearColor(
     Linear Color=&LerpLinearColors(StartColor=$Particles.Color, EndColor=<rgba 1 1 1 1>, LerpFactor=0.1)
     Scale Factor=&MakeFloat4FromVectorAndFloat(
      "Vector (RGB)"=&VectorFromFloat(Value=&MakeCustomFloatFromBool(Bool=$User.IsFrontEnd, True Float=2, False Float=2))
      "Float (Alpha)"=&MakeCustomFloatFromBool(Bool=$User.IsFrontEnd, True Float=1, False Float=0.5)
     )
    )
    Particles.SourceGlowScale:Vector2f=&Vector2DFromFloat(Value=40)
   ) {isSetParameters=true}
   /SetVariables(Particles.BackerColor:LinearColor=<rgba 0.1009 0.0268 0.0268 0.75>) {enabled=false, isSetParameters=true}
   &DynamicMaterialParameters(
    Index 0 Param 1=0.9
    Index 0 Param 2=&FloatFromCurve(
     FloatCurve=DataInterface() {propertyValues {Curve {keys [
         {interpMode=RCIM_Cubic, tangentMode=RCTM_User, time=0, value=0.95}
         {interpMode=RCIM_Cubic, tangentMode=RCTM_User, time=0.5, value=1}
         {tangentMode=RCTM_User, time=1, value=0.95}
        ]}}}
     CurveIndex=&Frac_Float(Float=&Add_Float(A=$Engine.System.Age, B=&RandomRangeFloat(Minimum=0, Maximum=1)))
    )
    Index 1 Param 1=&MakeCustomFloatFromBool(Bool=$User.IsFrontEnd, True Float=2, False Float=3)
    Module.Index0Float0Write=true
    Module.Index0Float1Write=true
    Module.Index0Float2Write=true
    Module.Index0Float3Write=true
    Module.Index1Float0Write=true
    Module.Index1Float1Write=true
    Module.Index1Float2Write=true
    Module.Index1Float3Write=true
    Write Parameter Index 1=::NewEnumerator1
   )
  ]
  rendererClasses [&NiagaraSpriteRendererProperties, &NiagaraLightRendererProperties]
  renderers [
   {rendererClass=&NiagaraSpriteRendererProperties}
   {rendererClass=&NiagaraSpriteRendererProperties, rendererIndex=1}
   {rendererClass=&NiagaraLightRendererProperties, rendererIndex=2}
   {rendererClass=&NiagaraSpriteRendererProperties, rendererIndex=3}
   {rendererClass=&NiagaraSpriteRendererProperties, rendererIndex=4}
   {rendererClass=&NiagaraSpriteRendererProperties, rendererIndex=5}
  ]
 }
}
summary {
 systemName=NS_Aura_Lights
 emitters [
  {emitterName=MainLight, rendererClasses [&NiagaraSpriteRendererProperties, &NiagaraLightRendererProperties]}
  {emitterName=FigureEightBeams, rendererClasses [&NiagaraSpriteRendererProperties]}
  {emitterName=Glows, rendererClasses [&NiagaraSpriteRendererProperties]}
  {emitterName=Body_Stars, rendererClasses [&NiagaraSpriteRendererProperties]}
  {emitterName=Dust, rendererClasses [&NiagaraSpriteRendererProperties]}
  {bEnabled=false, emitterName=BlackBacker, rendererClasses [&NiagaraSpriteRendererProperties]}
 ]
 userVariables [
  User.bIsFrontEnd {type=NiagaraBool}
  User.bIsFrontEndPreview {type=NiagaraBool}
  User.Interpolate Intensity {type=NiagaraFloat, defaultValue=5, description=""}
  User.IsFrontEnd {type=NiagaraBool}
  User.IsLargeBody {type=NiagaraBool}
 ]
}
