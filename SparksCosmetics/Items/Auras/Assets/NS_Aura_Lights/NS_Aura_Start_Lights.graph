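# One-shot Niagara start effect for the Lights aura cosmetic, bursting colored light around a character's skeletal mesh.
# Corona and Flare flash single color-cycling sprites, while BackBeams spawns short-lived rotating beams.
# Body_Stars places sprites between paired bones and drifts them away from a reference bone.
# IsFrontEnd switches beam and star colors and adjusts beam brightness, while IsLargeBody lengthens beams.
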NiagaraSystem /SparksCosmetics/Items/Auras/Assets/NS_Aura_Lights/NS_Aura_Start_Lights
dependencies {
 usedDataInterfaces [
  &NiagaraDataInterfaceColorCurve
  &NiagaraDataInterfaceCurve
  &NiagaraDataInterfaceVector2DCurve
  &NiagaraDataInterfaceArrayInt32
  &NiagaraDataInterfaceArrayFloat
  &NiagaraDataInterfaceSkeletalMesh
 ]
 usedDynamicInputs [&MakeVector, &SelectFloatFromArray]
 usedModules [
  &self:SystemSpawnScript.SystemScriptSource.SystemScriptGraph.NiagaraNodeAssignment_0.SetVariables_#1
  &SystemState
 ]
}
emitters {
 BackBeams {
  emitterUpdateScript [
   &EmitterState()
   &SpawnBurst_Instantaneous(Spawn Count=20, Use Spawn Probability=false, UseLoopCountLimit=false)
  ]
  particleSpawnScript [
   /SetVariables(Particles.Variance:NiagaraFloat=&FloatFromCurve(
     FloatCurve=DataInterface() {propertyValues {Curve {keys [{interpMode=RCIM_Cubic, tangentMode=RCTM_User, time=0, value=0}, {interpMode=RCIM_Cubic, tangentMode=RCTM_User, time=1, value=1}]}}}
     CurveIndex=&RandomRangeFloat(Minimum=0, Maximum=1)
    )) {isSetParameters=true}
   &InitializeParticle(
    Color=&MakeCustomLinearColorFromBool(
     Bool=$User.IsFrontEnd
     True Color=&ColorFromCurve(
      ColorCurve=DataInterface() {propertyValues {
        AlphaCurve {keys [<at 0 1 RCIM_Cubic>]}
        BlueCurve {keys [<at 0 1>, {arriveTangent=-2, interpMode=RCIM_Cubic, leaveTangent=-2, time=0.25, value=0}, <at 0.5 0>, <at 0.75 0>, <at 1 1 RCIM_Cubic>]}
        GreenCurve {keys [<at 0 0>, {arriveTangent=0.6, interpMode=RCIM_Cubic, leaveTangent=0.6, time=0.25, value=0}, <at 0.5 0.3>, <at 0.75 0.3>, <at 1 0 RCIM_Cubic>]}
        RedCurve {keys [<at 0 0>, {arriveTangent=0.6, interpMode=RCIM_Cubic, leaveTangent=0.6, time=0.25, value=0.6}, <at 0.5 0.3>, <at 0.75 0>, <at 1 0 RCIM_Cubic>]}
       }}
      CurveIndex=&RandomRangeFloat(Minimum=0, Maximum=1)
     )
     False Color=&MakeLinearColorFromVectorAndFloat(
      "Vector (RGB)"=&self:AbsVectorChannels(Input:Vector3f=&RandomVector(Vector Scale=1, Evaluation Type=::NewEnumerator1))
      "Float (Alpha)"=1
     )
    )
    Lifetime=&Lerp_Float(A=0.25, B=0.75, Alpha=$Particles.Variance)
    Sprite Rotation Mode=::NewEnumerator2
    Sprite Size=&Multiply_Vector2D(
     A=&Lerp_Vector2(A=<10 100>, B=<80 100>, Alpha=$Particles.Variance)
     B=&MakeCustomVector2DFromBool(Bool=$User.IsLargeBody, True Vector2D=<1 1.2>, False Vector2D=<1 1>)
    )
    Sprite Size Mode=::NewEnumerator4
    UsePositionOffset=false
    Write Material Random=true
   )
   /SetVariables(Particles.DebugSpriteScale:Vector2f=&Vector2DFromFloat(Value=5)) {enabled=false, isSetParameters=true}
   /SetVariables(Particles.PositionTargetA:NiagaraPosition=&RandomRangeVector(Minimum=<-50 -5 160>, Maximum=<50 -15 160>)) {isSetParameters=true}
   &SampleSkeletalMesh(
    AdditionalYawPitchRoll=false
    Bone / Socket Index=7
    Bone Sampling Mode:ENiagaraBoneSamplingMode=::NewEnumerator4
    Position Offset Coordinate Space:ENiagaraCoordinateSpace=::Local
    Skeletal Mesh:NiagaraDataInterfaceSkeletalMesh=$System.Skeletal Mesh
   )
  ]
  particleUpdateScript [
   &ParticleState()
   &ScaleSpriteSize(
    Non-Uniform Curve Index:NiagaraFloat=&Lerp_Float(A=$Particles.NormalizedAge, B=$Particles.Age, Alpha=$Particles.Variance)
    Non-Uniform Curve Sprite Scale=DataInterface() {propertyValues {
      XCurve {keys [{tangentMode=RCTM_User, time=0, value=0.5}, {tangentMode=RCTM_User, time=0.75, value=1}]}
      YCurve {keys [
        {arriveTangent=17.1041, interpMode=RCIM_Cubic, leaveTangent=17.1041, tangentMode=RCTM_User, time=0, value=0}
        {
         arriveTangent=-2.5141
         arriveTangentWeight=0.088
         interpMode=RCIM_Cubic
         leaveTangent=-2.5141
         leaveTangentWeight=0.1477
         tangentMode=RCTM_User
         time=0.2
         value=2
        }
        <at 0.75 1 RCIM_Cubic>
       ]}
     }}
    Scale Sprite Size Mode=::NewEnumerator3
   )
   &SpriteRotationRate(Rotation Rate=&Multiply_Float(
     A=&RandomRangeFloat(Minimum=-1, Maximum=1)
     B=&Lerp_Float(A=50, B=20, Alpha=$Particles.Variance)
    ))
   &ScaleColor(
    Scale Alpha=&FloatFromCurve(
     FloatCurve=DataInterface() {propertyValues {Curve {keys [
         {arriveTangent=6.0801, interpMode=RCIM_Cubic, leaveTangent=6.0801, tangentMode=RCTM_User, time=0, value=0}
         {interpMode=RCIM_Cubic, tangentMode=RCTM_User, time=0.5, value=1}
         {tangentMode=RCTM_User, time=1, value=0}
        ]}}}
     Scale Curve=&Multiply_Float(
      A=&RandomRangeFloat(Minimum=0.2, Maximum=0.5)
      B=&MakeCustomFloatFromBool(Bool=$User.IsFrontEnd, True Float=1, False Float=0.5)
     )
    )
    Scale RGB=&VectorFromFloat(Value=&FloatFromCurve(
      FloatCurve=DataInterface() {propertyValues {Curve {keys [{arriveTangent=-44.703, interpMode=RCIM_Cubic, leaveTangent=-44.703, tangentMode=RCTM_User, time=0, value=10}, <at 0.5 1>]}}}
      Scale Curve=&MakeCustomFloatFromBool(Bool=$User.IsFrontEnd, True Float=1, False Float=1.2)
     ))
    ScaleA=true
    ScaleRGB=true
   )
  ]
  rendererClasses [&NiagaraSpriteRendererProperties]
  renderers [
   {rendererClass=&NiagaraSpriteRendererProperties}
   {rendererClass=&NiagaraSpriteRendererProperties, rendererIndex=1}
  ]
 }
 Body_Stars {
  emitterUpdateScript [
   &EmitterState()
   &SpawnBurst_Instantaneous(Spawn Count=50, Use Spawn Probability=false, UseLoopCountLimit=false)
  ]
  particleSpawnScript [
   /SetVariables(Particles.PickBone:NiagaraInt32=&RandomRangeInt(Minimum=1, Maximum=6)) {isSetParameters=true}
   /SetVariables(Particles.PickBoneMatch:NiagaraInt32=&SelectIntFromArray(Direct Array Index=$Particles.PickBone, Int Selection Array=DataInterface() {propertyValues {IntData [0,0,0,1,2,3,4]}})) {isSetParameters=true}
   /SetVariables(Transient.Variance:NiagaraFloat=&RandomRangeFloat(Minimum=0, Maximum=1)) {isSetParameters=true}
   &SkeletalMeshLocation(
    Bone / Socket Index=$Particles.PickBone
    Bone Sampling Mode=::NewEnumerator4
    MaskSpawnGroup=false
    Orientation Sampling=::NewEnumerator1
    Position Sampling=::NewEnumerator2
    Sampled Position Offset=&RandomVector(Vector Scale=5)
    Sampled Position Offset Coordinate Space:ENiagaraSkelMeshTransforms=::NewEnumerator1
    Skeletal Mesh=$System.Skeletal Mesh
   )
   &SkeletalMeshLocation(
    Bone / Socket Index=$Particles.PickBoneMatch
    Bone Sampling Mode=::NewEnumerator4
    MaskSpawnGroup=false
    Orientation Sampling=::NewEnumerator1
    Position Sampling=::NewEnumerator2
    Sampled Position Offset=&RandomVector(Vector Scale=5)
    Sampled Position Offset Coordinate Space:ENiagaraSkelMeshTransforms=::NewEnumerator1
    Skeletal Mesh=$System.Skeletal Mesh
   ) {module=SkeletalMeshLocation001}
   &SkeletalMeshLocation(
    Bone / Socket Index=7
    Bone Sampling Mode=::NewEnumerator4
    MaskSpawnGroup=false
    Orientation Sampling=::NewEnumerator1
    Position Sampling=::NewEnumerator2
    Skeletal Mesh=$System.Skeletal Mesh
   ) {module=SkeletalMeshLocation002}
   &InitializeParticle(
    Color=&ColorFromCurve(
     ColorCurve=DataInterface() {propertyValues {
       AlphaCurve {keys [<at 0 1 RCIM_Cubic>]}
       BlueCurve {keys [<at 0 1>, {arriveTangent=-2, interpMode=RCIM_Cubic, leaveTangent=-2, time=0.25, value=0}, <at 0.5 0>, <at 0.75 0>, <at 1 1 RCIM_Cubic>]}
       GreenCurve {keys [<at 0 0>, {arriveTangent=0.6, interpMode=RCIM_Cubic, leaveTangent=0.6, time=0.25, value=0}, <at 0.5 0.3>, <at 0.75 0.3>, <at 1 0 RCIM_Cubic>]}
       RedCurve {keys [<at 0 0>, {arriveTangent=0.6, interpMode=RCIM_Cubic, leaveTangent=0.6, time=0.25, value=0.6}, <at 0.5 0.3>, <at 0.75 0>, <at 1 0 RCIM_Cubic>]}
      }}
     CurveIndex=&RandomRangeFloat(Minimum=0, Maximum=1)
    )
    Lifetime=&Lerp_Float(A=0.3, B=1.5, Alpha=$Transient.Variance)
    Mesh Scale Mode=::NewEnumerator3
    Mesh Uniform Scale=&RandomRangeFloat(Minimum=0.05, Maximum=0.25)
    Position=&LerpPosition(
     A=$Particles.SkeletalMeshLocation.SampledPosition
     B=$Particles.SkeletalMeshLocation001.SampledPosition
     Alpha=&RandomRangeFloat(Minimum=0, Maximum=1)
    )
    Position Mode=::NewEnumerator1
    Sprite Size Mode=::NewEnumerator3
    Uniform Sprite Size=&Lerp_Float(A=12, B=4, Alpha=$Transient.Variance)
    UsePositionOffset=false
    Write Material Random=true
   )
   &AddVelocity(
    Rotation Coordinate Space=::World
    Velocity=&Multiply_VectorByFloat(
     Vector=&NormalizeVector(
      Vector To Normalize=&Subtract_Position(A=$Particles.Position, B=$Particles.SkeletalMeshLocation002.SampledPosition)
      Safe Normalize=true
      Safe Fallback Threshold=0
      Safe Fallback Vector=<0 0 0.001>
     )
     Float=20
    )
    Velocity Speed Scale=&RandomRangeFloat(Minimum=0, Maximum=1.5)
   )
   /SetVariables(Particles.Color=&MakeCustomLinearColorFromBool(
     Bool=$User.IsFrontEnd
     True Color=$Particles.Color
     False Color=&MultiplyLinearColor(
      Linear Color=&HueShiftLinearColor(Hue=0, Saturation=2, Value=0.5, Linear Color=$Particles.Color)
      Scale Factor=&MakeFloat4FromVectorAndFloat("Vector (RGB)"=<1 1 1>, "Float (Alpha)"=5)
     )
    )) {isSetParameters=true}
   /SetVariables(Particles.Color=&ColorFromCurve(
     ColorCurve=DataInterface() {propertyValues {
       AlphaCurve {keys [<at 0 1 RCIM_Cubic>]}
       BlueCurve {keys [<at 0 1 RCIM_Cubic>,<at 0.02 0>,<at 0.98 1>,<at 1 0 RCIM_Cubic>]}
       GreenCurve {keys [<at 0 1 RCIM_Cubic>,<at 0.02 0>,<at 0.98 0>,<at 1 1 RCIM_Cubic>]}
       RedCurve {keys [<at 0 1 RCIM_Cubic>,<at 0.02 1>,<at 0.98 0>,<at 1 0 RCIM_Cubic>]}
      }}
     CurveIndex=$Transient.VelocityModifier
    )) {enabled=false, isSetParameters=true}
  ]
  particleUpdateScript [
   &CameraOffset(Camera Offset Amount=-25)
   &ParticleState()
   &SpriteRotationRate(Rotation Rate=&FloatFromCurve(
     FloatCurve=DataInterface() {propertyValues {Curve {keys [<at 0 1 RCIM_Cubic>, {arriveTangent=-2.1279, tangentMode=RCTM_User, time=1, value=0}]}}}
     Scale Curve=&RandomRangeFloat(Minimum=-2, Maximum=2)
    ))
   &ScaleSpriteSize(Uniform Curve Sprite Scale=DataInterface() {propertyValues {Curve {keys [
        {interpMode=RCIM_Cubic, tangentMode=RCTM_User, time=0, value=0.8}
        {interpMode=RCIM_Cubic, tangentMode=RCTM_User, time=0.5, value=1.5}
        {tangentMode=RCTM_User, time=1, value=0}
       ]}}})
   &Drag(Rotational Drag=0.1, Use Linear Drag=true, Use Rotational Drag=true)
   &SolveForcesAndVelocity(Clamp Velocity=false, Limit Acceleration=false)
  ]
  rendererClasses [&NiagaraSpriteRendererProperties]
  renderers [{rendererClass=&NiagaraSpriteRendererProperties}]
 }
 Corona {
  emitterUpdateScript [
   &EmitterState()
   &SpawnBurst_Instantaneous(Spawn Count=1, Use Spawn Probability=false)
  ]
  particleSpawnScript [
   &InitializeParticle(
    Lifetime=0.5
    Sprite Rotation Mode=::NewEnumerator2
    Sprite Size Mode=::NewEnumerator3
    Uniform Sprite Size=400
    UsePositionOffset=false
    Write Material Random=true
   )
   &SystemLocation(MaskSpawnGroup=false, Offset=<0 0 120>)
  ]
  particleUpdateScript [
   &ParticleState()
   &SolveForcesAndVelocity(Clamp Velocity=false, Limit Acceleration=false)
   &SkeletalMeshLocation(Bone / Socket Index=7, Bone Sampling Mode=::NewEnumerator4, MaskSpawnGroup=false, Skeletal Mesh=$System.Skeletal Mesh) {module=SkeletalMeshLocation001}
   &ScaleColor(
    Scale Alpha=&FloatFromCurve(FloatCurve=DataInterface() {propertyValues {Curve {keys [
         {interpMode=RCIM_Cubic, tangentMode=RCTM_User, time=0, value=0}
         {arriveTangent=-1.6778, interpMode=RCIM_Cubic, leaveTangent=-1.6778, tangentMode=RCTM_User, time=0.14, value=1}
         {arriveTangent=-1.1628, interpMode=RCIM_Cubic, leaveTangent=-1.1628, time=0.4229, value=0.3088}
         {interpMode=RCIM_Cubic, tangentMode=RCTM_User, time=1, value=0}
        ]}}})
    Scale RGB=&MakeVectorFromLinearColorRGB(Color=&ColorFromCurve(ColorCurve=DataInterface() {propertyValues {
        AlphaCurve {keys [<at 0 1 RCIM_Cubic>]}
        BlueCurve {keys [<at 0.2 0>,<at 0.5 0.4696>,<at 0.8 1 RCIM_Cubic>]}
        GreenCurve {keys [<at 0.2 0>,<at 0.5 2>,<at 0.8 0 RCIM_Cubic>]}
        RedCurve {keys [<at 0.2 3>,<at 0.5 0.4696>,<at 0.8 0 RCIM_Cubic>]}
       }}))
    ScaleA=true
    ScaleRGB=true
   )
   &CameraOffset(Camera Offset Amount=-30)
   &ScaleSpriteSize(Scale Factor=&Vector2DFromFloat(Value=&FloatFromCurve(FloatCurve=DataInterface() {propertyValues {Curve {keys [
          {arriveTangent=4.3497, interpMode=RCIM_Cubic, leaveTangent=4.3497, tangentMode=RCTM_User, time=0, value=0}
          {arriveTangent=0.7215, interpMode=RCIM_Cubic, leaveTangent=0.7215, tangentMode=RCTM_User, time=0.228, value=0.7588}
          {arriveTangent=0.1996, interpMode=RCIM_Cubic, leaveTangent=0.1996, tangentMode=RCTM_User, time=1, value=1}
         ]}}})))
   &DynamicMaterialParameters(
    Index 0 Param 1=0.75
    Index 0 Param 2=5
    Index 0 Param 3=1
    Index 0 Param 4=1
    Index 1 Param 1=6
    Index 1 Param 2=&FloatFromCurve(FloatCurve=DataInterface() {propertyValues {Curve {keys [
         {interpMode=RCIM_Cubic, tangentMode=RCTM_User, time=0.25, value=0}
         {arriveTangent=0.2003, interpMode=RCIM_Cubic, leaveTangent=0.2003, tangentMode=RCTM_User, time=1, value=1}
        ]}}})
    Index 1 Param 3=1
    Index 1 Param 4=1
    Module.Param0WriteEnabled=true
    Module.Param0WriteEnabled1=true
    Module.Param1WriteEnabled=true
    Module.Param1WriteEnabled1=true
    Module.Param2WriteEnabled=true
    Module.Param2WriteEnabled1=true
    Module.Param3WriteEnabled=true
    Module.Param3WriteEnabled1=true
    Write Parameter Index 1=true
   )
  ]
  rendererClasses [&NiagaraSpriteRendererProperties]
  renderers [{rendererClass=&NiagaraSpriteRendererProperties}]
 }
 Flare {
  emitterUpdateScript [
   &EmitterState()
   &SpawnBurst_Instantaneous(Spawn Count=1, Use Spawn Probability=false)
  ]
  particleSpawnScript [
   &InitializeParticle(
    Lifetime=0.3
    Sprite Rotation Mode=::NewEnumerator2
    Sprite Size Mode=::NewEnumerator3
    Uniform Sprite Size=600
    UsePositionOffset=false
    Write Material Random=true
   )
   &DynamicMaterialParameters(
    Index 0 Param 1=&RandomRangeFloat(Minimum=0.4, Maximum=0.6)
    Index 0 Param 2=1
    Index 0 Param 3=1
    Index 0 Param 4=1
    Module.Param0WriteEnabled=true
    Module.Param1WriteEnabled=true
    Module.Param2WriteEnabled=true
    Module.Param3WriteEnabled=true
   )
   &SystemLocation(MaskSpawnGroup=false, Offset=<0 0 120>)
   /SetVariables(Particles.Sharp Flare Camera Offset=40) {isSetParameters=true}
   &CameraOffset(Camera Offset Amount=-30) {module=CameraOffset001}
  ]
  particleUpdateScript [
   &ParticleState()
   &SolveForcesAndVelocity(Clamp Velocity=false, Limit Acceleration=false)
   &SkeletalMeshLocation(Bone / Socket Index=7, Bone Sampling Mode=::NewEnumerator4, MaskSpawnGroup=false, Skeletal Mesh=$System.Skeletal Mesh) {module=SkeletalMeshLocation001}
   &ScaleColor(
    Scale Alpha=&FloatFromCurve(FloatCurve=DataInterface() {propertyValues {Curve {keys [<at 0 0>,<at 0.2 1>,<at 0.8 1>,<at 1 0>]}}})
    Scale RGB=&MakeVectorFromLinearColorRGB(Color=&ColorFromCurve(ColorCurve=DataInterface() {propertyValues {
        AlphaCurve {keys [<at 0 1 RCIM_Cubic>]}
        BlueCurve {keys [<at 0.2 0>,<at 0.5 0>,<at 0.8 1 RCIM_Cubic>]}
        GreenCurve {keys [<at 0.2 0>,<at 0.5 1>,<at 0.8 0 RCIM_Cubic>]}
        RedCurve {keys [<at 0.2 1>,<at 0.5 0>,<at 0.8 0 RCIM_Cubic>]}
       }}))
    ScaleA=true
    ScaleRGB=true
   )
   &ScaleSpriteSize(Scale Factor=&Vector2DFromFloat(Value=&FloatFromCurve(FloatCurve=DataInterface() {propertyValues {Curve {keys [<at 0 1.5 RCIM_Cubic>,<at 1 0>]}}})))
   /SetVariables(Particles.Flash color:LinearColor=&MakeLinearColorFromVectorAndFloat(
     "Vector (RGB)"=&MakeVectorFromLinearColorRGB(Color=&ColorFromCurve(ColorCurve=DataInterface() {propertyValues {
         AlphaCurve {keys [<at 0 1 RCIM_Cubic>]}
         BlueCurve {keys [<at 0.12 0>,<at 0.3 0>,<at 0.48 1 RCIM_Cubic>]}
         GreenCurve {keys [<at 0.12 0>,<at 0.3 1>,<at 0.48 0 RCIM_Cubic>]}
         RedCurve {keys [<at 0.12 1>,<at 0.3 0>,<at 0.48 0 RCIM_Cubic>]}
        }}))
     "Float (Alpha)"=&Multiply_Float(
      A=&Multiply_Float(
       A=&FNTimeOfDayScalar(Night Scale=1, Morning Scale=1, Evening Scale=1)
       B=&FloatFromCurve(FloatCurve=DataInterface() {propertyValues {Curve {keys [<at 0 0>,<at 0.124 1>,<at 0.496 1>,<at 0.62 0>]}}})
      )
      B=1
     )
    )) {isSetParameters=true}
   /SetVariables(Particles.Flash size:Vector2f=&Multiply_Vector2D_ByFloat(
     Float=&Multiply_Float(
      A=400
      B=&FloatFromCurve(FloatCurve=DataInterface() {propertyValues {Curve {keys [
           {arriveTangent=4.1785, interpMode=RCIM_Cubic, leaveTangent=4.1785, tangentMode=RCTM_User, time=0, value=0}
           {arriveTangent=-0.15, interpMode=RCIM_Cubic, leaveTangent=-0.15, tangentMode=RCTM_User, time=0.1378, value=1}
           {arriveTangent=-0.0405, interpMode=RCIM_Cubic, leaveTangent=-0.0405, tangentMode=RCTM_User, time=0.62, value=0.75}
          ]}}})
     )
     Vector2D=<1 1>
    )) {isSetParameters=true}
  ]
  rendererClasses [&NiagaraSpriteRendererProperties]
  renderers [
   {rendererClass=&NiagaraSpriteRendererProperties}
   {rendererClass=&NiagaraSpriteRendererProperties, rendererIndex=1}
  ]
 }
}
summary {
 systemName=NS_Aura_Start_Lights
 emitters [
  {emitterName=Corona, rendererClasses [&NiagaraSpriteRendererProperties]}
  {emitterName=Flare, rendererClasses [&NiagaraSpriteRendererProperties]}
  {emitterName=BackBeams, rendererClasses [&NiagaraSpriteRendererProperties]}
  {emitterName=Body_Stars, rendererClasses [&NiagaraSpriteRendererProperties]}
 ]
 userVariables [
  User.IsFrontEnd {type=NiagaraBool, defaultValue=true}
  User.IsLargeBody {type=NiagaraBool}
 ]
}
